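# Looping spot ambience cue for the Doom Forge at the CastleForge POI, mixing fire rumble with thunder and lightning.
# Fire and rock rumble loops split into close and medium attenuations, with a whoosh sweetener added up close.
# Lightning and thunder beds join three looped, randomly delayed and modulated one-shot layers, which are shared by close and far attenuations.
# All quality levels share one mix, and concurrency allows a single instance.
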SoundCue /Game/Sounds/Helios/POI/CastleForge/Spots/Spot_DoomForge_Fire_Smoke_Lightning_Loop_Cue
root=QualityLevel_0
nodes [
 {
  id=Attenuation_0
  children [Mixer_6]
  properties {attenuationSettings=&/Castle_Forge_Main_Fire_Rumble_Close_Dist_Att}
 }
 {
  id=Attenuation_1
  children [Modulator_0]
  properties {attenuationSettings=&/Castle_Forge_Main_Fire_Rumble_Med_Dist_Focus_Att}
 }
 {
  id=Attenuation_2
  children [WavePlayer_10]
  properties {attenuationSettings=&/Doom_Forge_Thunder_Far_Distant_Max_Focus_Att}
 }
 {
  id=Attenuation_3
  children [Modulator_4]
  properties {attenuationSettings=&/Castle_Forge_Thunder_Lightning_Close_Dist_Att}
 }
 {
  id=Attenuation_4
  children [Mixer_4]
  properties {attenuationSettings=&/Doom_Forge_Thunder_Far_Distant_Focus_Att}
 }
 {
  id=Attenuation_5
  children [Mixer_4]
  properties {attenuationSettings=&/Doom_Forge_Thunder_Close_Focus_Att}
 }
 {id=Delay_0, children [Mixer_1], properties {delayMax=0.2, delayMin=0.2}}
 {id=Delay_1, children [Random_1], properties {delayMax=8, delayMin=3}}
 {id=Delay_2, children [Random_2], properties {delayMax=3, delayMin=2}}
 {id=Delay_3, children [Random_0], properties {delayMax=3, delayMin=2}}
 {id=Looping_0, children [Modulator_1]}
 {id=Looping_1, children [Modulator_3]}
 {id=Looping_2, children [Modulator_2]}
 {id=Mixer_0, children [Attenuation_0, Attenuation_1], properties {inputVolume [2,1]}}
 {id=Mixer_1, children [WavePlayer_8,WavePlayer_9], properties {inputVolume [1,1]}}
 {id=Mixer_2, children [Attenuation_3, Attenuation_2, Mixer_7], properties {inputVolume [1,1.5,2]}}
 {id=Mixer_3, children [Mixer_0,Mixer_2], properties {inputVolume [1,1]}}
 {id=Mixer_4, children [Looping_1,Looping_2,Looping_0], properties {inputVolume [1,1,1]}}
 {id=Mixer_6, children [Modulator_0,Modulator_6], properties {inputVolume [1,2]}}
 {id=Mixer_7, children [Attenuation_4, Attenuation_5], properties {inputVolume [1,1]}}
 {id=Modulator_0, children [Delay_0], properties {pitchMax=1, pitchMin=1, volumeMax=1, volumeMin=1}}
 {id=Modulator_1, children [Delay_1], properties {pitchMax=1.1, pitchMin=0.9, volumeMax=1}}
 {id=Modulator_2, children [Delay_2], properties {pitchMax=1.1, pitchMin=0.93, volumeMax=1.2}}
 {id=Modulator_3, children [Delay_3], properties {pitchMax=1.2, pitchMin=1, volumeMax=1}}
 {id=Modulator_4, children [WavePlayer_26], properties {pitchMax=1.2, pitchMin=1, volumeMax=1}}
 {id=Modulator_6, children [WavePlayer_11], properties {pitchMax=1.1, pitchMin=1.1}}
 {id=QualityLevel_0, children [Mixer_3,Mixer_3,Mixer_3,Mixer_3]}
 {id=Random_0, children [WavePlayer_0,WavePlayer_2,WavePlayer_3,WavePlayer_4,WavePlayer_6,WavePlayer_7], properties {weights [1,1,1,1,1,1]}}
 {id=Random_1, children [WavePlayer_5, WavePlayer_13, WavePlayer_14], properties {weights [1,1,1]}}
 {id=Random_2, children [WavePlayer_21, WavePlayer_22, WavePlayer_23, WavePlayer_24, WavePlayer_27], properties {weights [1,1,1,1,1]}}
 {id=WavePlayer_0, properties {soundWaveAssetPtr=/Game/Sounds/Asteria/POI/Meteor/Spots/Spot_Meteor_Thunder_Soft_Rumbling_Distant_01}}
 {
  id=WavePlayer_10
  properties {bLooping=true, soundWaveAssetPtr=&/Spot_Doom_Forge_Thunder_Rumble_Loop}
 }
 {
  id=WavePlayer_11
  properties {bLooping=true, soundWaveAssetPtr=/Game/Sounds/Helios/Environment_Actors/Brazier/Spot_Fire_Huge_Whooshy_Sweetener_Loop}
 }
 {id=WavePlayer_13, properties {soundWaveAssetPtr=/Game/Sounds/Asteria/POI/Meteor/Spots/Spot_Meteor_Thunder_Crack_Swt_03}}
 {id=WavePlayer_14, properties {soundWaveAssetPtr=/Game/Sounds/Asteria/POI/Meteor/Spots/Spot_Meteor_Thunder_Crack_Swt_04}}
 {id=WavePlayer_2, properties {soundWaveAssetPtr=/Game/Sounds/Asteria/POI/Meteor/Spots/Spot_Meteor_Thunder_Soft_Rumbling_Distant_02}}
 {id=WavePlayer_21, properties {soundWaveAssetPtr=/Game/Sounds/Asteria/POI/Meteor/Spots/Spot_Meteor_Thunder_Soft_Rumbling_Distant_07}}
 {id=WavePlayer_22, properties {soundWaveAssetPtr=/Game/Sounds/Asteria/POI/Meteor/Spots/Spot_Meteor_Thunder_Soft_Rumbling_Distant_08}}
 {id=WavePlayer_23, properties {soundWaveAssetPtr=/Game/Sounds/Asteria/POI/Meteor/Spots/Spot_Meteor_Thunder_Soft_Rumbling_Distant_09}}
 {id=WavePlayer_24, properties {soundWaveAssetPtr=/Game/Sounds/Asteria/POI/Meteor/Spots/Spot_Meteor_Thunder_Soft_Rumbling_Distant_10}}
 {
  id=WavePlayer_26
  properties {bLooping=true, soundWaveAssetPtr=/Game/Sounds/Asteria/POI/Meteor/Spots/Spot_Meteor_Lightning_Cracks_Close_Loop_02}
 }
 {id=WavePlayer_27, properties {soundWaveAssetPtr=/Game/Sounds/Asteria/POI/Meteor/Spots/Spot_Meteor_Thunder_Soft_Rumbling_Distant_11}}
 {id=WavePlayer_3, properties {soundWaveAssetPtr=/Game/Sounds/Asteria/POI/Meteor/Spots/Spot_Meteor_Thunder_Soft_Rumbling_Distant_03}}
 {id=WavePlayer_4, properties {soundWaveAssetPtr=/Game/Sounds/Asteria/POI/Meteor/Spots/Spot_Meteor_Thunder_Soft_Rumbling_Distant_04}}
 {id=WavePlayer_5, properties {soundWaveAssetPtr=/Game/Sounds/Asteria/POI/Meteor/Spots/Spot_Meteor_Thunder_Crack_Swt_01}}
 {id=WavePlayer_6, properties {soundWaveAssetPtr=/Game/Sounds/Asteria/POI/Meteor/Spots/Spot_Meteor_Thunder_Soft_Rumbling_Distant_05}}
 {id=WavePlayer_7, properties {soundWaveAssetPtr=/Game/Sounds/Asteria/POI/Meteor/Spots/Spot_Meteor_Thunder_Soft_Rumbling_Distant_06}}
 {
  id=WavePlayer_8
  properties {bLooping=true, soundWaveAssetPtr=&/Spot_Doom_Forge_Fire_Rock_Desctruction_Loop}
 }
 {id=WavePlayer_9, properties {bLooping=true, soundWaveAssetPtr=&/Spot_Doom_Forge_Main_Rumble_Loop}}
]
properties {
 bOverrideConcurrency=true
 maxDistance=43000
 soundClassObject=/Game/Sounds/FortSoundClassesAndModes/Fort_Storms_PSM
 volumeMultiplier=0.5
 attenuationOverrides {
  attenuationShape=Capsule
  bAttenuateWithLPF=true
  bEnableLogFrequencyScaling=true
  falloffDistance=10000
  lPFFrequencyAtMax=197.5057
  lPFFrequencyAtMin=20511.4102
  lPFRadiusMax=3789.3618
  lPFRadiusMin=1822.3112
  nonSpatializedRadiusEnd=2500
  nonSpatializedRadiusStart=5000
  stereoSpread=2000
  attenuationShapeExtents=<512.3855 1207.2642>
 }
 concurrencyOverrides {maxCount=1, resolutionRule=PreventNew}
 soundSubmixSends [{minSendLevel=0.3, soundSubmix=&Fortnite_Reverb_Default}]
}
