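# Interior ambient spot cue for the CastleForge POI's Doom Forge cloud, built from a reused meteor rumble loop and wind one-shots.
# After a short delay, two envelopers slowly drift pitch and volume over a quality switch.
# The first quality branch mixes the attenuated rumble with two delayed, looping chains of randomized, modulated gust, rustle and alpine layers.
# The other quality branches share a rumble-only mixer.
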SoundCue /Game/Sounds/Helios/POI/CastleForge/Spots/Spot_Doom_Forge_Cloud_Int_Cue
root=Delay_0
nodes [
 {
  id=Attenuation_2
  children [WavePlayer_5]
  properties {attenuationSettings=/Game/Sounds/Asteria/POI/Meteor/Spots/Meteor_Main_Rumble_Capsule_Inside_Att}
 }
 {
  id=Attenuation_4
  children [WavePlayer_12]
  properties {attenuationSettings=&Meteor_Main_Rumble_Capsule_Inside_Att}
 }
 {id=Delay_0, children [Enveloper_1], properties {delayMax=0.5}}
 {id=Delay_1, children [Random_5], properties {delayMax=7, delayMin=2}}
 {id=Delay_2, children [Random_10], properties {delayMax=5, delayMin=2}}
 {
  id=Enveloper_0
  children [QualityLevel_0]
  properties {
   loopEnd=30
   pitchCurve {editorCurveData {keys [{time=0}]}}
   volumeCurve {editorCurveData {keys [<at 0 1 RCIM_Cubic>,<at 15 0.8 RCIM_Cubic>,<at 30 1>]}}
  }
 }
 {
  id=Enveloper_1
  children [Enveloper_0]
  properties {
   loopEnd=45
   pitchCurve {editorCurveData {keys [{interpMode=RCIM_Cubic, time=0}, <at 22.5 1.2 RCIM_Cubic>, {time=45}]}}
   volumeCurve {editorCurveData {keys [<at 0 1>]}}
  }
 }
 {id=Looping_0, children [Delay_2]}
 {id=Looping_1, children [Delay_1]}
 {id=Mixer_0, children [Modulator_4,Modulator_1], properties {inputVolume [0.5,1.25]}}
 {id=Mixer_1, children [Attenuation_2, Looping_1, Looping_0], properties {inputVolume [1,1,1]}}
 {id=Mixer_2, children [Modulator_3,Modulator_4], properties {inputVolume [1.25,0.5]}}
 {id=Mixer_3, children [Attenuation_4], properties {inputVolume [1]}}
 {id=Mixer_5, children [Modulator_2,Modulator_6], properties {inputVolume [1,0.5]}}
 {id=Mixer_6, children [Modulator_6,Modulator_0], properties {inputVolume [0.5,1]}}
 {id=Modulator_0, children [Random_3], properties {pitchMax=1, pitchMin=0.8, volumeMax=1.1, volumeMin=0.9}}
 {id=Modulator_1, children [Random_0], properties {pitchMax=1, pitchMin=0.8, volumeMax=1.1, volumeMin=0.9}}
 {id=Modulator_2, children [Random_8], properties {pitchMax=1.1, pitchMin=0.8, volumeMax=1.1, volumeMin=0.9}}
 {id=Modulator_3, children [Random_2], properties {pitchMax=1.1, pitchMin=0.8, volumeMax=1.1, volumeMin=0.9}}
 {id=Modulator_4, children [Random_4], properties {pitchMax=1, pitchMin=0.8, volumeMax=1.1, volumeMin=0.9}}
 {id=Modulator_6, children [Random_9], properties {pitchMax=1, pitchMin=0.8, volumeMax=1.1, volumeMin=0.9}}
 {id=QualityLevel_0, children [Mixer_1,Mixer_3,Mixer_3,Mixer_3]}
 {
  id=Random_0
  children [WavePlayer_0, WavePlayer_1, WavePlayer_2, WavePlayer_20, WavePlayer_27, WavePlayer_28, WavePlayer_29, WavePlayer_30]
  properties {weights [1,1,1,1,1,1,1,1]}
 }
 {id=Random_10, children [Mixer_5,Mixer_6], properties {weights [1,1]}}
 {id=Random_2, children [WavePlayer_9, WavePlayer_11, WavePlayer_3], properties {weights [1,1,1]}}
 {
  id=Random_3
  children [WavePlayer_4, WavePlayer_6, WavePlayer_7, WavePlayer_26, WavePlayer_31, WavePlayer_32, WavePlayer_33, WavePlayer_34]
  properties {weights [1,1,1,1,1,1,1,1]}
 }
 {id=Random_4, children [WavePlayer_17, WavePlayer_18, WavePlayer_19, WavePlayer_21, WavePlayer_22], properties {weights [1,1,1,1,1]}}
 {id=Random_5, children [Mixer_2,Mixer_0], properties {weights [1,1]}}
 {id=Random_8, children [WavePlayer_8, WavePlayer_13, WavePlayer_14], properties {weights [1,1,1]}}
 {id=Random_9, children [WavePlayer_15, WavePlayer_16, WavePlayer_23, WavePlayer_24, WavePlayer_25], properties {weights [1,1,1,1,1]}}
 {id=WavePlayer_0, properties {soundWaveAssetPtr=/Game/Sounds/Artemis/Biomes/Mountains/Mountain_Spots/Spot_Mountain_Day_Gusty_Wind_06}}
 {id=WavePlayer_1, properties {soundWaveAssetPtr=/Game/Sounds/Artemis/Biomes/Mountains/Mountain_Spots/Spot_Mountain_Day_Gusty_Wind_01}}
 {id=WavePlayer_11, properties {soundWaveAssetPtr=/Game/Sounds/Apollo/Biomes/Alpine/Biome_Alpine_Oneshot_Day_Quad_03}}
 {id=WavePlayer_12, properties {bLooping=true, soundWaveAssetPtr=/Game/Sounds/Asteria/POI/Meteor/Spots/Spot_Meteor_Main_Rumble_Loop}}
 {id=WavePlayer_13, properties {soundWaveAssetPtr=&Biome_Alpine_Oneshot_Day_Quad_03}}
 {id=WavePlayer_14, properties {soundWaveAssetPtr=/Game/Sounds/Apollo/Biomes/Alpine/Biome_Alpine_Oneshot_Day_Quad_04}}
 {id=WavePlayer_15, properties {soundWaveAssetPtr=/Game/Sounds/Artemis/Biomes/Grasslands/Grassland_Spots/Spot_Grassland_Wind_Rustle_01}}
 {id=WavePlayer_16, properties {soundWaveAssetPtr=/Game/Sounds/Artemis/Biomes/Grasslands/Grassland_Spots/Spot_Grassland_Wind_Rustle_02}}
 {id=WavePlayer_17, properties {soundWaveAssetPtr=&Spot_Grassland_Wind_Rustle_01}}
 {id=WavePlayer_18, properties {soundWaveAssetPtr=&Spot_Grassland_Wind_Rustle_02}}
 {id=WavePlayer_19, properties {soundWaveAssetPtr=/Game/Sounds/Artemis/Biomes/Grasslands/Grassland_Spots/Spot_Grassland_Wind_Rustle_03}}
 {id=WavePlayer_2, properties {soundWaveAssetPtr=/Game/Sounds/Artemis/Biomes/Mountains/Mountain_Spots/Spot_Mountain_Day_Gusty_Wind_02}}
 {id=WavePlayer_20, properties {soundWaveAssetPtr=/Game/Sounds/Fort_Ambient_NonLoop/Desert/Wind_DesertGust_Whistle14}}
 {id=WavePlayer_21, properties {soundWaveAssetPtr=/Game/Sounds/Artemis/Biomes/Grasslands/Grassland_Spots/Spot_Grassland_Wind_Rustle_05}}
 {id=WavePlayer_22, properties {soundWaveAssetPtr=/Game/Sounds/Artemis/Biomes/Grasslands/Grassland_Spots/Spot_Grassland_Wind_Rustle_06}}
 {id=WavePlayer_23, properties {soundWaveAssetPtr=&Spot_Grassland_Wind_Rustle_03}}
 {id=WavePlayer_24, properties {soundWaveAssetPtr=&Spot_Grassland_Wind_Rustle_05}}
 {id=WavePlayer_25, properties {soundWaveAssetPtr=&Spot_Grassland_Wind_Rustle_06}}
 {id=WavePlayer_26, properties {soundWaveAssetPtr=&Wind_DesertGust_Whistle14}}
 {id=WavePlayer_27, properties {soundWaveAssetPtr=/Game/Sounds/Fort_Ambient_NonLoop/Desert/Wind_DesertGust_Whistle15}}
 {id=WavePlayer_28, properties {soundWaveAssetPtr=/Game/Sounds/Fort_Ambient_NonLoop/Desert/Wind_DesertGust_Whistle16}}
 {id=WavePlayer_29, properties {soundWaveAssetPtr=/Game/Sounds/Fort_Ambient_NonLoop/Desert/Wind_DesertGust_Whistle18}}
 {id=WavePlayer_3, properties {soundWaveAssetPtr=&Biome_Alpine_Oneshot_Day_Quad_04}}
 {id=WavePlayer_30, properties {soundWaveAssetPtr=/Game/Sounds/Fort_Ambient_NonLoop/Desert/Wind_DesertGust_Whistle26}}
 {id=WavePlayer_31, properties {soundWaveAssetPtr=&Wind_DesertGust_Whistle15}}
 {id=WavePlayer_32, properties {soundWaveAssetPtr=&Wind_DesertGust_Whistle16}}
 {id=WavePlayer_33, properties {soundWaveAssetPtr=&Wind_DesertGust_Whistle18}}
 {id=WavePlayer_34, properties {soundWaveAssetPtr=&Wind_DesertGust_Whistle26}}
 {id=WavePlayer_4, properties {soundWaveAssetPtr=&Spot_Mountain_Day_Gusty_Wind_06}}
 {id=WavePlayer_5, properties {bLooping=true, soundWaveAssetPtr=&Spot_Meteor_Main_Rumble_Loop}}
 {id=WavePlayer_6, properties {soundWaveAssetPtr=&Spot_Mountain_Day_Gusty_Wind_01}}
 {id=WavePlayer_7, properties {soundWaveAssetPtr=&Spot_Mountain_Day_Gusty_Wind_02}}
 {id=WavePlayer_8, properties {soundWaveAssetPtr=/Game/Sounds/Apollo/Biomes/Alpine/Biome_Alpine_Oneshot_Day_Quad_01}}
 {id=WavePlayer_9, properties {soundWaveAssetPtr=&Biome_Alpine_Oneshot_Day_Quad_01}}
]
properties {
 attenuationSettings=&Fort_Ambient_Athena_Att
 bOverrideConcurrency=true
 maxDistance=2097152
 soundClassObject=/Game/Sounds/FortSoundClassesAndModes/Fort_Storms_PSM
 volumeMultiplier=1
 attenuationOverrides {
  bAttenuateWithLPF=true
  bEnableLogFrequencyScaling=true
  distanceAlgorithm=Logarithmic
  falloffDistance=10000
  lPFFrequencyAtMax=197.5057
  lPFFrequencyAtMin=20511.4102
  lPFRadiusMax=3789.3618
  lPFRadiusMin=1822.3112
  stereoSpread=2000
  attenuationShapeExtents {x=0}
 }
 concurrencyOverrides {maxCount=2}
}
