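# Sound cue for an interior spot ambience sweetener at the CastleForge Doom Forge, playing distant structural stress sounds reused from the Meteor POI.
# A random node picks one of seven equally weighted waves, passes it through a modulator, and every quality level uses that same branch.
# Overridden attenuation adds distance filtering and stereo spread, with a low concurrency cap and a distance-based send to the default reverb submix.
# It belongs to the outdoor ambient sound class.
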SoundCue /Game/Sounds/Helios/POI/CastleForge/Spots/Spot_Doom_Forge_Cloud_Sweeteners_Int_Cue
root=QualityLevel_0
nodes [
 {id=Modulator_2, children [Random_0]}
 {id=QualityLevel_0, children [Modulator_2,Modulator_2,Modulator_2,Modulator_2]}
 {
  id=Random_0
  children [WavePlayer_8, WavePlayer_9, WavePlayer_10, WavePlayer_16, WavePlayer_17, WavePlayer_18, WavePlayer_19]
  properties {weights [1,1,1,1,1,1,1]}
 }
 {id=WavePlayer_10, properties {soundWaveAssetPtr=/Game/Sounds/Asteria/POI/Meteor/Spots/Spot_Meteor_Structural_Stress_Distant_03}}
 {id=WavePlayer_16, properties {soundWaveAssetPtr=/Game/Sounds/Asteria/POI/Meteor/Spots/Spot_Meteor_Structural_Stress_Distant_04}}
 {id=WavePlayer_17, properties {soundWaveAssetPtr=/Game/Sounds/Asteria/POI/Meteor/Spots/Spot_Meteor_Structural_Stress_Distant_05}}
 {id=WavePlayer_18, properties {soundWaveAssetPtr=/Game/Sounds/Asteria/POI/Meteor/Spots/Spot_Meteor_Structural_Stress_Distant_06}}
 {id=WavePlayer_19, properties {soundWaveAssetPtr=/Game/Sounds/Asteria/POI/Meteor/Spots/Spot_Meteor_Structural_Stress_Distant_07}}
 {id=WavePlayer_8, properties {soundWaveAssetPtr=/Game/Sounds/Asteria/POI/Meteor/Spots/Spot_Meteor_Structural_Stress_Distant_01}}
 {id=WavePlayer_9, properties {soundWaveAssetPtr=/Game/Sounds/Asteria/POI/Meteor/Spots/Spot_Meteor_Structural_Stress_Distant_02}}
]
properties {
 bOverrideAttenuation=true
 bOverrideConcurrency=true
 maxDistance=9000
 soundClassObject=&Fort_Ambient_Outdoors_Athena
 volumeMultiplier=1.504
 attenuationOverrides {
  bEnableLogFrequencyScaling=true
  falloffDistance=6000
  hPFFrequencyAtMax=1500
  hPFFrequencyAtMin=750
  lPFFrequencyAtMin=7500
  lPFRadiusMax=8000
  stereoSpread=500
  attenuationShapeExtents {x=3000}
 }
 concurrencyOverrides {maxCount=2, voiceStealReleaseTime=2}
 soundSubmixSends [{maxSendDistance=6000, minSendDistance=3000, minSendLevel=0.6, soundSubmix=&Fortnite_Reverb_Default}]
}
