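# Sound Cue for the looping lava pool ambience at the Doom forge spot in the CastleForge POI, reusing Athena volcano caldera waves.
# A Mixer layers a looping gurgle bed with louder flare one-shots, which a Looping node retriggers through random delay, volume/pitch modulation and random selection.
# An Enveloper ramps the mix's volume up from silence.
# Overridden attenuation adds distance filtering and listener focus, quieting the pool when off-focus.
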SoundCue /Game/Sounds/Helios/POI/CastleForge/Spots/Spot_Doom_Forge_Lava_Pool_Loop_Cue
root=Enveloper_0
nodes [
 {id=Delay_0, children [Modulator_0], properties {delayMax=3, delayMin=1}}
 {
  id=Enveloper_0
  children [Mixer_0]
  properties {pitchCurve {editorCurveData {keys [{time=0}]}}, volumeCurve {editorCurveData {keys [<at 0 0>,<at 1 1>]}}}
 }
 {id=Looping_0, children [Delay_0]}
 {id=Mixer_0, children [WavePlayer_3,Looping_0], properties {inputVolume [1,2]}}
 {id=Modulator_0, children [Random_0], properties {volumeMax=1.2}}
 {
  id=Random_0
  children [
   WavePlayer_1
   WavePlayer_2
   WavePlayer_5
   WavePlayer_6
   WavePlayer_7
   WavePlayer_8
   WavePlayer_9
   WavePlayer_10
   WavePlayer_11
   WavePlayer_12
   WavePlayer_13
   WavePlayer_14
   WavePlayer_15
  ]
  properties {weights [1,1,1,1,1,1,1,1,1,1,1,1,1]}
 }
 {id=WavePlayer_1, properties {soundWaveAssetPtr=/Game/Athena/Sounds/Ambient/Volcano/AMB_Lava_Caldera_Spots_Flare_13}}
 {id=WavePlayer_10, properties {soundWaveAssetPtr=/Game/Athena/Sounds/Ambient/Volcano/AMB_Lava_Caldera_Spots_Flare_06}}
 {id=WavePlayer_11, properties {soundWaveAssetPtr=/Game/Athena/Sounds/Ambient/Volcano/AMB_Lava_Caldera_Spots_Flare_05}}
 {id=WavePlayer_12, properties {soundWaveAssetPtr=/Game/Athena/Sounds/Ambient/Volcano/AMB_Lava_Caldera_Spots_Flare_04}}
 {id=WavePlayer_13, properties {soundWaveAssetPtr=/Game/Athena/Sounds/Ambient/Volcano/AMB_Lava_Caldera_Spots_Flare_03}}
 {id=WavePlayer_14, properties {soundWaveAssetPtr=/Game/Athena/Sounds/Ambient/Volcano/AMB_Lava_Caldera_Spots_Flare_02}}
 {id=WavePlayer_15, properties {soundWaveAssetPtr=/Game/Athena/Sounds/Ambient/Volcano/AMB_Lava_Caldera_Spots_Flare_01}}
 {id=WavePlayer_2, properties {soundWaveAssetPtr=/Game/Athena/Sounds/Ambient/Volcano/AMB_Lava_Caldera_Spots_Flare_12}}
 {id=WavePlayer_3, properties {bLooping=true, soundWaveAssetPtr=/Game/Athena/Sounds/Ambient/Volcano/AMB_Lava_Caldera_Gurgle_Close_2D_Loop}}
 {id=WavePlayer_5, properties {soundWaveAssetPtr=/Game/Athena/Sounds/Ambient/Volcano/AMB_Lava_Caldera_Spots_Flare_11}}
 {id=WavePlayer_6, properties {soundWaveAssetPtr=/Game/Athena/Sounds/Ambient/Volcano/AMB_Lava_Caldera_Spots_Flare_10}}
 {id=WavePlayer_7, properties {soundWaveAssetPtr=/Game/Athena/Sounds/Ambient/Volcano/AMB_Lava_Caldera_Spots_Flare_09}}
 {id=WavePlayer_8, properties {soundWaveAssetPtr=/Game/Athena/Sounds/Ambient/Volcano/AMB_Lava_Caldera_Spots_Flare_08}}
 {id=WavePlayer_9, properties {soundWaveAssetPtr=/Game/Athena/Sounds/Ambient/Volcano/AMB_Lava_Caldera_Spots_Flare_07}}
]
properties {
 bOverrideAttenuation=true
 maxDistance=1850
 soundClassObject=&Fort_Gadgets
 volumeMultiplier=0.5
 attenuationOverrides {
  bAttenuateWithLPF=true
  bEnableFocusInterpolation=true
  bEnableListenerFocus=true
  falloffDistance=1250
  focusAttackInterpSpeed=4
  focusAzimuth=15
  hPFFrequencyAtMax=300
  lPFFrequencyAtMax=600
  lPFRadiusMax=1000
  lPFRadiusMin=800
  nonFocusAzimuth=50
  nonFocusVolumeAttenuation=0.2
  attenuationShapeExtents {x=600}
 }
 concurrencyOverrides {maxCount=1, resolutionRule=PreventNew}
}
